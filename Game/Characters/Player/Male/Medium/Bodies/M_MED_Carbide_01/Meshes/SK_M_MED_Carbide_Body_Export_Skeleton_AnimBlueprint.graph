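# Animation Blueprint for the M_MED_Carbide_01 male player body, which copies a pose from another mesh and adds procedural elbow pad motion.
# On initialization it caches the result of GetOwnerSkeletalMeshComponent as MeshToCopy, the source for Copy Pose From Mesh.
# In component space, two chained Bone Driven Controllers counter-rotate each dyn_elbowPad bone on Z by half its lowerarm's Z rotation.
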AnimBlueprint /Game/Characters/Player/Male/Medium/Bodies/M_MED_Carbide_01/Meshes/SK_M_MED_Carbide_Body_Export_Skeleton_AnimBlueprint
target_skeleton=&/SK_M_MED_Carbide_Body_Export_Skeleton
graphs [
 AnimGraph {nodes [
   n1 = Get MeshToCopy() -> (MeshToCopy:Skeletal Mesh Component Object Reference)
   n2 = "dyn_elbowPad_l.rotateZ = lowerarm_l.rotateZ * -0.5 - Bone Driven Controller"!(ComponentPose:Component Space Pose Link Structure=Local To Component(LocalPose=Copy Pose From Mesh(SourceMeshComponent=@n1)), Alpha:float=1) -> (Pose:Component Space Pose Link Structure) {
    is=title AnimGraphNode_BoneDrivenController
    bAffectTargetRotationZ=true
    multiplier=-0.5
    sourceComponent=RotationZ
    alphaBoolBlend {blendOption=Linear}
    sourceBone {boneName=lowerarm_l}
    targetBone {boneName=dyn_elbowPad_l}
   }
   n3 = "dyn_elbowPad_r.rotateZ = lowerarm_r.rotateZ * -0.5 - Bone Driven Controller"!(ComponentPose:Component Space Pose Link Structure=@n2, Alpha:float=1) -> (Pose:Component Space Pose Link Structure) {
    is=title AnimGraphNode_BoneDrivenController
    bAffectTargetRotationZ=true
    multiplier=-0.5
    sourceComponent=RotationZ
    alphaBoolBlend {blendOption=Linear}
    sourceBone {boneName=lowerarm_r}
    targetBone {boneName=dyn_elbowPad_r}
   }
   Output Pose(Result=Component To Local(ComponentPose=@n3)) {layerGroup=None}
  ]}
 EventGraph {nodes [
   Event BlueprintInitializeAnimation()
   > Set MeshToCopy(MeshToCopy:Skeletal Mesh Component Object Reference=GetOwnerSkeletalMeshComponent()) -> (Output_Get:Skeletal Mesh Component Object Reference)
   Event BlueprintUpdateAnimation()
  ]}
]
variables [MeshToCopy {type=SkeletalMeshComponent}]
